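; Music encounter Blueprint for Coffee Puff that links the encounter music logic to the Coffee Puff spawner actor.
; The AddMusicEncounterComponent override casts the component's owner to the spawner and initializes from it on success, then calls the parent in either case.
; Initialization binds Set Encounter State to the spawner's state-change event, then stores its root component and boss spawn location.
; It finally applies the spawner's current encounter state, with validity checks guarding these steps.
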(fn AddMusicEncounterComponent
  (params (Music%20Encounter%20Component :type Music%20Encounter%20Component%20Base%20Object%20Reference%20%28by%20ref%29))
  (bind _asbga_coffee_puff_spawner (Utilities|Casting|CastToBGA_CoffeePuff_Spawner (Components|GetOwner Music%20Encounter%20Component))
    (:then
      (Default|InitializeSpawnerComponent _asbga_coffee_puff_spawner)
      (|CallParentFunction Music%20Encounter%20Component)
      (return))
    (:CastFailed
      (|CallParentFunction Music%20Encounter%20Component)
      (return))))

(fn InitializeSpawnerComponent
  (params (Spawner :type BGA%20Coffee%20Puff%20Spawner%20Object%20Reference))
  (Utilities|IsValid Spawner
    (:Is%20Valid
      (Default|BindEventtoOnEncounterStateChanged Spawner (EventDispatchers|CreateEvent :!event Set%20Encounter%20State))
      (Utilities|FlowControl|Sequence
        (:then_0
          (bind _spawner (Variables|GetSpawner)
            (:then
              (Variables|Default|SetSpawner%20Root%20Component (Variables|/Script/Engine.Actor|GetRootComponent _spawner)))
            (:else)))
        (:then_1
          (Utilities|FlowControl|Sequence
            (:then_0
              (bind _spawner_2 (Variables|GetSpawner)
                (:then
                  (Variables|MusicEncounter|SetBoss%20Spawn%20Location (Variables|/VividRazor_CoffeePuff/GameplayActors/BGA_CoffeePuff_Spawner.BGA_CoffeePuff_Spawner_C|GetBoss%20Spawn%20Location _spawner_2)))
                (:else)))
            (:then_1
              (Utilities|FlowControl|Sequence
                (:then_0
                  (bind _spawner_3 (Variables|GetSpawner)
                    (:then
                      (Class|BGACoffeePuffSpawner|SetEncounterState (Variables|/VividRazor_CoffeePuff/GameplayActors/BGA_CoffeePuff_Spawner.BGA_CoffeePuff_Spawner_C|GetEncounter%20State _spawner_3)))
                    (:else)))))))))))
