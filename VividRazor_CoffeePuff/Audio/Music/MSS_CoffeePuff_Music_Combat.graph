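# Stereo MetaSound for CoffeePuff combat music, staged by the Stage input as intro stinger, looping combat bed and ending stinger that finishes the source.
# A BPM-driven metronome schedules transitions and retriggers the paired Perc and Else loops every set number of bars.
# AlertLevel crossfades per-layer volumes for Unaware, Suspicious and Threatened.
# An envelope eases the bed down after entry, and trigger gates filter Stage changes.
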MetaSoundSource /VividRazor_CoffeePuff/Audio/Music/MSS_CoffeePuff_Music_Combat
inputs [
 AlertLevel {type=Int32}
 AlertMixCrossfade {type=Time}
 BPM {type=Float}
 Combat Loop Length Bars {type=Int32}
 Loop Length Bars {type=Int32}
 MasterVolume {type=Float}
 Sounds.CombatIntro {type=WaveAsset}
 Sounds.CombatLoop.Else {type=WaveAsset}
 Sounds.CombatLoop.Perc {type=WaveAsset}
 Sounds.EndingStinger {type=WaveAsset}
 Stage {type=Int32}
 UE.Source.OnPlay {type=Trigger}
 Volume.Combat.Else.Suspicious {type=Float, access=Constructor}
 Volume.Combat.Else.Threatened {type=Float}
 Volume.Combat.Else.Unaware {type=Float}
 Volume.Combat.Perc.Suspicious {type=Float, access=Constructor}
 Volume.Combat.Perc.Threatened {type=Float}
 Volume.Combat.Perc.Unaware {type=Float}
]
interfaces [UE.OutputFormat.Stereo:1.0, UE.Source.OneShot:1.0, UE.Source:1.1]
outputs [
 "UE.OutputFormat.Stereo.Audio:0" {type=Audio}
 "UE.OutputFormat.Stereo.Audio:1" {type=Audio}
 UE.Source.OneShot.OnFinished {type=Trigger}
]
pages [Default {
  input_defaults {
   AlertLevel=2
   AlertMixCrossfade=1.4
   BPM=134
   Combat Loop Length Bars=16
   Loop Length Bars=16
   MasterVolume=1
   Sounds.CombatIntro=&/MX_S34_CoffeePuff_Approach_To_Combat_STNG
   Sounds.CombatLoop.Else=&/MX_S34_CoffeePuff_Combat_LP_Else
   Sounds.CombatLoop.Perc=&/MX_S34_CoffeePuff_Combat_LP_Perc
   Sounds.EndingStinger=&/MX_S34_CoffeePuff_Combat_END_STNG
   Stage=0
   UE.Source.OnPlay=null
   Volume.Combat.Else.Suspicious=0.25
   Volume.Combat.Else.Threatened=1
   Volume.Combat.Else.Unaware=0
   Volume.Combat.Perc.Suspicious=1
   Volume.Combat.Perc.Threatened=1
   Volume.Combat.Perc.Unaware=1
  }
  nodes [
   n1 = HarmonixNodes.MidiClockSubdivisionTrigger(MIDI Clock=HarmonixNodes.Metronome(Tempo=^BPM, Transport=HarmonixNodes.TriggerToTransport(Play=^UE.Source.OnPlay).Transport).MIDI Clock) {also {defaults {Base Grid Size=0}}}
   n2 = VariableAccessor.Trigger(Variable=VariableMutator.Trigger(Value=@n1.Trigger Out, Variable=InitVariable.Trigger().Variable).Variable)
   &MSP_ScheduleTrigger(Schedule=VariableAccessor.Trigger(Variable=InitVariable.Trigger().Variable).Value, In Trigger=@n2.Value)
   n3 = VariableAccessor.Trigger(Variable=@n2.Variable)
   n4 = InitVariable.Trigger()
   n5 = VariableDeferredAccessor.Trigger(Variable=@n4.Variable)
   n6 = VariableDeferredAccessor.Trigger(Variable=@n5.Variable)
   n7 = InitVariable.Trigger()
   n8 = VariableDeferredAccessor.Trigger(Variable=@n7.Variable)
   n9 = VariableDeferredAccessor.Trigger(Variable=@n8.Variable)
   TriggerAnyTriggerAny2_7 = "TriggerAny.Trigger Any (2)"(In 1=@n6.Value, In 0=@n9.Value)
   TriggerOnValueChange_0 = UE.Trigger On Value Change:Int32(Value=^Stage)
   n10 = VariableDeferredAccessor.Trigger(Variable=InitVariable.Trigger().Variable)
   TriggerControl_3 = UE.Trigger Control(Close=@TriggerAnyTriggerAny2_7.Out, Trigger In=@TriggerOnValueChange_0.Trigger, Open=@n10.Value)
   TriggerAnyTriggerAny2_6 = "TriggerAny.Trigger Any (2)"(In 1=^UE.Source.OnPlay, In 0=@TriggerControl_3.Trigger Out)
   n11 = /FortMetasound/Patches/MSP_SwitchOnInt_4(Int=^Stage, Compare=@TriggerAnyTriggerAny2_6.Out)
   TriggerAnyTriggerAny2_1 = "TriggerAny.Trigger Any (2)"(In 1=VariableDeferredAccessor.Trigger(Variable=@n7.Variable).Value, In 0=VariableDeferredAccessor.Trigger(Variable=@n4.Variable).Value)
   TriggerControl_1 = UE.Trigger Control(Trigger In=@n11.Out 0, Close=@TriggerAnyTriggerAny2_1.Out)
   n12 = VariableMutator.Trigger(Variable=InitVariable.Trigger().Variable, Value=@TriggerControl_1.Trigger Out)
   TriggerAnyTriggerAny3_0 = "TriggerAny.Trigger Any (3)"(In 2=@n8.Value, In 0=VariableDeferredAccessor.Trigger(Variable=@n4.Variable).Value, In 1=@n5.Value)
   TriggerControl_2 = UE.Trigger Control(Trigger In=@n11.Out 1, Close=@TriggerAnyTriggerAny3_0.Out)
   TriggerAnyTriggerAny2_9 = "TriggerAny.Trigger Any (2)"(
    In 0=&MSP_ScheduleTrigger(In Trigger=@n3.Value, Schedule=VariableAccessor.Trigger(Variable=@n12.Variable).Value) {also {defaults {Triggers From Now=4}}}.Scheduled Trigger
    In 1=@TriggerControl_2.Trigger Out
   )
   n13 = VariableAccessor.Trigger(Variable=VariableMutator.Trigger(Value=@TriggerAnyTriggerAny2_9.Out, Variable=@n10.Variable).Variable)
   n14 = VariableAccessor.Trigger(Variable=@n13.Variable)
   n15 = VariableAccessor.Trigger(Variable=@n14.Variable)
   n16 = VariableAccessor.Trigger(Variable=@n15.Variable)
   ADSREnvelopeADSREnvelope_0 = ADSR Envelope.ADSR Envelope:Float(Trigger Attack=VariableAccessor.Trigger(Variable=@n16.Variable).Value) {also {defaults {Decay Time=13, Release Time=0, Sustain Level=0.55}}}
   TriggerOnValueChange_1 = UE.Trigger On Value Change:Int32(Value=^AlertLevel)
   n17 = &MSP_SwitchOnInt_4(Int=^AlertLevel, Compare=@TriggerOnValueChange_1.Trigger)
   n18 = VariableAccessor.Trigger(Variable=VariableMutator.Trigger(Variable=InitVariable.Trigger().Variable, Value=@n17.Out 1).Variable)
   n19 = VariableAccessor.Trigger(Variable=@n18.Variable)
   n20 = VariableAccessor.Trigger(Variable=@n19.Variable)
   n21 = VariableAccessor.Trigger(Variable=VariableMutator.Trigger(Variable=InitVariable.Trigger().Variable, Value=@n17.Out 2).Variable)
   n22 = VariableAccessor.Trigger(Variable=@n21.Variable)
   n23 = VariableAccessor.Trigger(Variable=VariableMutator.Trigger(Variable=InitVariable.Trigger().Variable, Value=@n17.Out 0).Variable)
   n24 = VariableAccessor.Trigger(Variable=@n23.Variable)
   TriggerAnyTriggerAny2_8 = "TriggerAny.Trigger Any (2)"(In 1=@n22.Value, In 0=@n24.Value)
   n25 = VariableAccessor.Trigger(Variable=VariableMutator.Trigger(Variable=InitVariable.Trigger().Variable, Value=@TriggerAnyTriggerAny2_8.Out).Variable)
   n26 = Value.Value:Float(
    Target Value=^Volume.Combat.Perc.Suspicious
    Set=VariableAccessor.Trigger(Variable=@n20.Variable).Value
    Reset=VariableAccessor.Trigger(Variable=@n25.Variable).Value
   ) {also {defaults {Init Value=1}}}
   n27 = VariableAccessor.Trigger(Variable=@n24.Variable)
   TriggerAnyTriggerAny2_0 = "TriggerAny.Trigger Any (2)"(In 1=@n19.Value, In 0=@n27.Value)
   n28 = VariableAccessor.Trigger(Variable=VariableMutator.Trigger(Variable=InitVariable.Trigger().Variable, Value=@TriggerAnyTriggerAny2_0.Out).Variable)
   n29 = VariableAccessor.Trigger(Variable=@n22.Variable)
   n30 = Value.Value:Float(
    Target Value=^Volume.Combat.Perc.Threatened
    Reset=VariableAccessor.Trigger(Variable=@n28.Variable).Value
    Set=VariableAccessor.Trigger(Variable=@n29.Variable).Value
   ) {also {defaults {Init Value=1}}}
   TriggerAnyTriggerAny2_2 = "TriggerAny.Trigger Any (2)"(In 0=@n18.Value, In 1=@n21.Value)
   n31 = VariableAccessor.Trigger(Variable=VariableMutator.Trigger(Variable=InitVariable.Trigger().Variable, Value=@TriggerAnyTriggerAny2_2.Out).Variable)
   n32 = Value.Value:Float(
    Target Value=^Volume.Combat.Perc.Unaware
    Set=VariableAccessor.Trigger(Variable=@n27.Variable).Value
    Reset=VariableAccessor.Trigger(Variable=@n31.Variable).Value
   ) {also {defaults {Init Value=1}}}
   TriggerRouteTriggerRouteFloat3_1 = "TriggerRoute.Trigger Route (Float, 3):Float"(Set 1=@n26.On Set, Value 1=@n26.Output Value, Set 2=@n30.On Set, Value 2=@n30.Output Value, Set 0=@n32.On Set, Value 0=@n32.Output Value)
   n33 = VariableAccessor.Float(Variable=VariableMutator.Float(
     Variable=InitVariable.Float().Variable
     Value=UE.InterpTo:Audio(Interp Time=^AlertMixCrossfade, Target=@TriggerRouteTriggerRouteFloat3_1.Value).Value
    ).Variable)
   n34 = InitVariable.Trigger()
   n35 = VariableAccessor.Trigger(Variable=VariableMutator.Trigger(Variable=InitVariable.Trigger().Variable, Value=@n11.Out 2).Variable)
   n36 = VariableAccessor.Trigger(Variable=@n3.Variable)
   n37 = VariableAccessor.Trigger(Variable=VariableMutator.Trigger(Value=&MSP_ScheduleTrigger(Schedule=@n35.Value, In Trigger=@n36.Value).Scheduled Trigger, Variable=@n9.Variable).Variable)
   n38 = VariableAccessor.Trigger(Variable=@n37.Variable)
   n39 = VariableAccessor.Trigger(Variable=@n38.Variable)
   TriggerControl_0 = UE.Trigger Control(Trigger In=@n1.Trigger Out, Open=@n16.Value, Close=VariableAccessor.Trigger(Variable=@n39.Variable).Value)
   TriggerCounter_0 = UE.Trigger Counter(In=@TriggerControl_0.Trigger Out)
   n40 = TriggerCompare.Clamp:Int32(
    A=UE.Modulo:Int32(AdditionalOperands=^Combat Loop Length Bars, PrimaryOperand=Convert.Int32:Float(Float=@TriggerCounter_0.Value).Int32).Out
    Compare=@TriggerCounter_0.On Trigger
   ) {also {defaults {B=0}}}
   TriggerAnyTriggerAny2_3 = "TriggerAny.Trigger Any (2)"(In 0=@n13.Value, In 1=VariableAccessor.Trigger(Variable=VariableMutator.Trigger(Variable=@n34.Variable, Value=@n40.True).Variable).Value)
   n41 = VariableAccessor.Trigger(Variable=VariableMutator.Trigger(Value=@TriggerControl_1.Trigger Out, Variable=@n6.Variable).Variable)
   TriggerAnyTriggerAny2_5 = "TriggerAny.Trigger Any (2)"(In 1=@n39.Value, In 0=@n41.Value)
   n42 = VariableAccessor.Trigger(Variable=VariableMutator.Trigger(Variable=InitVariable.Trigger().Variable, Value=@TriggerAnyTriggerAny2_5.Out).Variable)
   WavePlayer_2 = UE.Wave Player:Stereo(Wave Asset=^Sounds.CombatLoop.Perc, Play=@TriggerAnyTriggerAny2_3.Out, Stop=@n42.Value)
   n43 = Value.Value:Float(Target Value=^Volume.Combat.Else.Unaware, Reset=@n31.Value, Set=@n23.Value) {also {defaults {Init Value=1}}}
   n44 = Value.Value:Float(Target Value=^Volume.Combat.Else.Threatened, Set=@n29.Value, Reset=@n28.Value) {also {defaults {Init Value=1}}}
   n45 = Value.Value:Float(Target Value=^Volume.Combat.Else.Suspicious, Set=@n20.Value, Reset=@n25.Value) {also {defaults {Init Value=1}}}
   TriggerRouteTriggerRouteFloat3_0 = "TriggerRoute.Trigger Route (Float, 3):Float"(Set 0=@n43.On Set, Value 0=@n43.Output Value, Set 2=@n44.On Set, Value 2=@n44.Output Value, Set 1=@n45.On Set, Value 1=@n45.Output Value)
   n46 = VariableAccessor.Float(Variable=VariableMutator.Float(
     Variable=InitVariable.Float().Variable
     Value=UE.InterpTo:Audio(Interp Time=^AlertMixCrossfade, Target=@TriggerRouteTriggerRouteFloat3_0.Value).Value
    ).Variable)
   WavePlayer_0 = UE.Wave Player:Stereo(Wave Asset=^Sounds.CombatLoop.Else, Play=@TriggerAnyTriggerAny2_3.Out, Stop=@n42.Value)
   n47 = VariableAccessor.Trigger(Variable=@n41.Variable)
   TriggerDelay_0 = UE.Trigger Delay(In=VariableAccessor.Trigger(Variable=@n47.Variable).Value) {also {defaults {Delay Time=0.45}}}
   WavePlayer_1 = UE.Wave Player:Stereo(Wave Asset=^Sounds.CombatIntro, Play=@TriggerDelay_0.Out)
   WavePlayer_3 = UE.Wave Player:Stereo(Wave Asset=^Sounds.EndingStinger, Play=@n37.Value)
   AudioMixerAudioMixerStereo3_0 = "AudioMixer.Audio Mixer (Stereo, 3)"(
    Gain 1=@ADSREnvelopeADSREnvelope_0.Out Envelope
    In 1 R=UE.Add:Audio(
     AdditionalOperands=UE.Multiply:Audio by Float(AdditionalOperands=@n33.Value, PrimaryOperand=@WavePlayer_2.Out Right).Out
     PrimaryOperand=UE.Multiply:Audio by Float(AdditionalOperands=@n46.Value, PrimaryOperand=@WavePlayer_0.Out Right).Out
    ).Out
    In 1 L=UE.Add:Audio(
     AdditionalOperands=UE.Multiply:Audio by Float(AdditionalOperands=@n33.Value, PrimaryOperand=@WavePlayer_2.Out Left).Out
     PrimaryOperand=UE.Multiply:Audio by Float(AdditionalOperands=@n46.Value, PrimaryOperand=@WavePlayer_0.Out Left).Out
    ).Out
    In 0 L=@WavePlayer_1.Out Left
    In 0 R=@WavePlayer_1.Out Right
    In 2 L=@WavePlayer_3.Out Left
    In 2 R=@WavePlayer_3.Out Right
   ) {also {defaults {Gain 0=1.35, Gain 2=0.6}}}
   Multiply_2 = UE.Multiply:Audio by Float(AdditionalOperands=^MasterVolume, PrimaryOperand=@AudioMixerAudioMixerStereo3_0.Out R)
   Multiply_5 = UE.Multiply:Audio by Float(AdditionalOperands=^MasterVolume, PrimaryOperand=@AudioMixerAudioMixerStereo3_0.Out L)
   TriggerAnyTriggerAny2_4 = "TriggerAny.Trigger Any (2)"(In 0=@n14.Value, In 1=@n38.Value)
   VariableMutator.Trigger(Variable=InitVariable.Trigger().Variable, Value=@TriggerAnyTriggerAny2_4.Out)
   TriggerAnyTriggerAny2_10 = "TriggerAny.Trigger Any (2)"(In 0=@n47.Value, In 1=@n15.Value)
   VariableMutator.Trigger(Variable=InitVariable.Trigger().Variable, Value=@TriggerAnyTriggerAny2_10.Out)
  ]
  links [
   {dst="@UE.OutputFormat.Stereo.Audio:1", src=Multiply_2.Out}
   {dst="@UE.OutputFormat.Stereo.Audio:0", src=Multiply_5.Out}
   {dst="@UE.Source.OneShot.OnFinished", src=WavePlayer_3.On Finished}
  ]
 }]
properties {
 assetImportData=&self:AssetImportData
 attenuationSettings=&/ATT_CoffeePuff_Music
 bOverrideConcurrency=true
 maxDistance=14000
 numChannels=2
 outputFormat=Stereo
 sampleRate=48000
 soundClassObject=&Fort_Music_NPC
 virtualizationMode=PlayWhenSilent
 volume=0.8
 concurrencyOverrides {maxCount=1}
}
