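# Stereo MetaSound music source for CoffeePuff that loops a supplied wave on a tempo grid, with attenuation and MusicEncounter pause support.
# A BPM-driven repeat trigger feeds beat and bar counters, which retrigger the wave player through a trigger variable every loop length in bars.
# Pause changes drive an interpolated gain multiplied by master volume, and a resettable delay matching the fade time fires OnFinished.
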MetaSoundSource /VividRazor_CoffeePuff/Audio/Music/MSS_CoffeePuff_MusicLoop
inputs [
 Beats per Bar {type=Int32}
 BPM {type=Float}
 Loop Length Bars {type=Int32}
 MasterVolume {type=Float}
 MusicEncounter.bIsPaused {type=Bool}
 MusicEncounter.PauseFadeOutTime {type=Time}
 UE.Source.OnPlay {type=Trigger}
 Wave Asset {type=WaveAsset}
]
interfaces [UE.OutputFormat.Stereo:1.0, UE.Source.OneShot:1.0, UE.Source:1.1]
outputs [
 "UE.OutputFormat.Stereo.Audio:0" {type=Audio}
 "UE.OutputFormat.Stereo.Audio:1" {type=Audio}
 UE.Source.OneShot.OnFinished {type=Trigger}
]
pages [Default {
  input_defaults {
   BPM=120
   Beats per Bar=4
   Loop Length Bars=16
   MasterVolume=1
   MusicEncounter.PauseFadeOutTime=2
   MusicEncounter.bIsPaused=null
   UE.Source.OnPlay=null
   Wave Asset=null
  }
  nodes [
   TriggerOnValueChange_0 = UE.Trigger On Value Change:Bool(Value=^MusicEncounter.bIsPaused)
   n1 = TriggerCompare.Clamp:Bool(A=^MusicEncounter.bIsPaused, Compare=@TriggerOnValueChange_0.Trigger) {also {defaults {B="<MetaSoundBuilderResult.SUCCEEDED: 0>"}}}
   n2 = Value.Value:Float(Reset=@n1.False, Set=@n1.True) {also {defaults {Init Value=1}}}
   n3 = UE.Multiply:Float(
    PrimaryOperand=^MasterVolume
    AdditionalOperands=UE.InterpTo:Audio(Interp Time=^MusicEncounter.PauseFadeOutTime, Target=@n2.Output Value).Value
   )
   n4 = InitVariable.Trigger()
   TriggerCounter_1 = UE.Trigger Counter(Reset Count=^Beats per Bar, In=UE.TriggerRepeat(Start=^UE.Source.OnPlay, Period=UE.BPMToSeconds(BPM=^BPM).Seconds).RepeatOut)
   TriggerCounter_0 = UE.Trigger Counter(In=TriggerCompare.Clamp:Int32(A=UE.Modulo:Int32(AdditionalOperands=^Beats per Bar, PrimaryOperand=@TriggerCounter_1.Count).Out, Compare=@TriggerCounter_1.On Trigger) {also {defaults {B=1}}}.True)
   n5 = TriggerCompare.Clamp:Int32(
    A=UE.Modulo:Int32(AdditionalOperands=^Loop Length Bars, PrimaryOperand=Convert.Int32:Float(Float=@TriggerCounter_0.Value).Int32).Out
    Compare=@TriggerCounter_0.On Trigger
   )
   TriggerAnyTriggerAny2_0 = "TriggerAny.Trigger Any (2)"(
    In 0=^UE.Source.OnPlay
    In 1=VariableAccessor.Trigger(Variable=VariableMutator.Trigger(Variable=@n4.Variable, Value=@n5.True).Variable).Value
   )
   WavePlayer_0 = UE.Wave Player:Stereo(Wave Asset=^Wave Asset, Play=@TriggerAnyTriggerAny2_0.Out)
   Multiply_0 = UE.Multiply:Audio by Float(AdditionalOperands=@n3.Out, PrimaryOperand=@WavePlayer_0.Out Left)
   Multiply_1 = UE.Multiply:Audio by Float(AdditionalOperands=@n3.Out, PrimaryOperand=@WavePlayer_0.Out Right)
   TriggerDelay_0 = UE.Trigger Delay(Delay Time=^MusicEncounter.PauseFadeOutTime, Reset=@n2.On Reset, In=@n2.On Set)
  ]
  links [
   {dst="@UE.OutputFormat.Stereo.Audio:0", src=Multiply_0.Out}
   {dst="@UE.OutputFormat.Stereo.Audio:1", src=Multiply_1.Out}
   {dst="@UE.Source.OneShot.OnFinished", src=TriggerDelay_0.Out}
  ]
 }]
properties {
 assetImportData=&self:AssetImportData
 attenuationSettings=&/ATT_CoffeePuff_Music
 maxDistance=14000
 numChannels=2
 outputFormat=Stereo
 sampleRate=48000
 soundClassObject=&Fort_Music_NPC
}
