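# MetaSound Source preset that supplies the looping idle music for CoffeePuff, using the Approach loop wave and classed as NPC music.
# It inherits its graph from MSS_CoffeePuff_MusicLoop and sets the wave, tempo, bar-based loop length, volume and encounter pause fade-out inputs.
# Output is stereo with a dedicated idle attenuation asset.
# A concurrency override allows one instance, and virtualization keeps it playing when silent.
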MetaSoundSource /VividRazor_CoffeePuff/Audio/Music/MSS_CoffeePuff_Music_Idle
preset_parent=&/MSS_CoffeePuff_MusicLoop
inputs [
 Beats per Bar {type=Int32}
 BPM {type=Float}
 Loop Length Bars {type=Int32}
 MasterVolume {type=Float}
 MusicEncounter.bIsPaused {type=Bool}
 MusicEncounter.PauseFadeOutTime {type=Time}
 UE.Source.OnPlay {type=Trigger}
 Wave Asset {type=WaveAsset}
]
interfaces [UE.OutputFormat.Stereo:1.0, UE.Source.OneShot:1.0, UE.Source:1.1]
outputs [
 "UE.OutputFormat.Stereo.Audio:0" {type=Audio}
 "UE.OutputFormat.Stereo.Audio:1" {type=Audio}
 UE.Source.OneShot.OnFinished {type=Trigger}
]
pages [Default {input_defaults {
   BPM=101
   Beats per Bar=4
   Loop Length Bars=16
   MasterVolume=0.7
   MusicEncounter.PauseFadeOutTime=3
   MusicEncounter.bIsPaused=null
   UE.Source.OnPlay=null
   Wave Asset=&/MX_S34_CoffeePuff_Approach_LP
  }}]
properties {
 assetImportData=&self:AssetImportData
 attenuationSettings=&/ATT_CoffeePuff_Music_Idle
 bOverrideConcurrency=true
 maxDistance=11500
 numChannels=2
 outputFormat=Stereo
 sampleRate=48000
 soundClassObject=&Fort_Music_NPC
 virtualizationMode=PlayWhenSilent
 volume=0.4
 concurrencyOverrides {maxCount=1}
}
